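; UI director for DemonMaskVoid teleport marker messages, controlling a message scene for the owning local player.
; On BeginPlay, a zone check gates a backwards-compatible stateful listener on the global event router, whose payload feeds HandleMessage, which filters by message owner.
; ShowMessageScene adds or removes the scene only on state changes.
; Expiry rechecks for the best active message and EndPlay clears its timer, though much selection and resend logic is empty here.
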(event EventBeginPlay
  (if (GlobalUIContext|IsinZone (PlayerController|LocalPlayerSubsystems|GetFortGlobalUIContext (DynamicUIDirector|GetOwningLocalPlayerController)))
    (bind (_asynctaskproxy _payload) (GameplayEventRouter|StartListeningToStatefulEventBackwardsCompatible "(TagName=\"EventMessage.DemonMaskVoid.TeleportMarker\")" (GameplayEventRouter|GetGlobalEventRouter) "/DemonMaskVoid/UI/TeleportMarker/STRUCT_DemonMaskVoidTeleportMarkerMessage.STRUCT_DemonMaskVoidTeleportMarkerMessage")
      (:OnEventReceived
        _)
      (:OnSavedEventStateExists
        _))
    (CallFunction|HandleMessage _payload)))

(event EventEndPlay
  (params (EndPlayReason :type EEndPlayReason%20Enum :default "Destroyed"))
  (Utilities|Time|ClearandInvalidateTimerbyHandle (Variables|Default|GetActiveMessageTimerHandle)))

(fn CheckAndResendActiveMessage
  (bind _isactivemessagefound (CallFunction|FindBestActiveMessage))
  (if _isactivemessagefound
    _))

(fn FindActiveMessageOfTypeFromList
  (params (MessageIdentifier :type Gameplay%20Tag%20Container%20Structure%20%28by%20ref%29) (MessageList :type STRUCT%20Active%20Demon%20Mask%20Void%20Teleport%20Marker%20Message%20Data%20Structure%20%28by%20ref%29)))

(fn FindBestActiveMessage
  (LocalVariables|SetCurrentGameTime (Utilities|Time|GetGameTimeinSeconds)))

(fn HandleMessage
  (params (MessageData :type STRUCT%20Demon%20Mask%20Void%20Teleport%20Marker%20Message%20Structure%20%28by%20ref%29))
  (bind (_messagedata_33_4559c4d74175241d33f5e4beb73de1a8 _sourceteleportmarker_35_fe723e534182e3ac71f904aa3cd35dd6 _showmessage_25_5a673aaa4467d28ffd71d690f6246f24 _messageowner_32_eff5db5f4c941f6915b5e882552f3052) (Utilities|Struct|BreakSTRUCTDemonMaskVoidTeleportMarkerMessage MessageData))
  (if (== (DynamicUIDirector|GetOwningLocalPlayerController) _messageowner_32_eff5db5f4c941f6915b5e882552f3052)
    (LocalVariables|SetNewMessage _messagedata_33_4559c4d74175241d33f5e4beb73de1a8)
    (LocalVariables|SetSourceTeleportMarker _sourceteleportmarker_35_fe723e534182e3ac71f904aa3cd35dd6)
    (LocalVariables|SetIsCurrentMessagesValid 0)
    (if _showmessage_25_5a673aaa4467d28ffd71d690f6246f24
      _
      (elif (LocalVariables|GetIsCurrentMessagesValid)
        _))))

(fn OnActiveMessageExpired
  (CallFunction|CheckandResendActiveMessage))

(fn ShowMessageScene
  (params (ShouldShow :type Boolean))
  (bind _messagescene (Variables|Default|GetMessageScene))
  (if (!= ShouldShow (Variables|Default|GetIsMessageSceneShowing))
    (bind _output_get (Variables|Default|SetIsMessageSceneShowing ShouldShow))
    (if _output_get
      (DynamicUIDirector|AddScene _messagescene)
      (else
        (DynamicUIDirector|RemoveScene _messagescene)))))

(fn ConstructionScript)
