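# SC_VoidMask_UI_Marker_ExpireWarning is a Sound Cue for the Void Mask UI, named as the warning for a marker about to expire.
# It routes through the Fort_UI_InGame_Critical sound class and the CS_VoidMask concurrency set.
# The cue has no nodes and a null root, so it holds no audio source and plays nothing as it stands.
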SoundCue /DemonMaskVoid/Assets/Audio/SC_VoidMask_UI_Marker_ExpireWarning
root=null
nodes []
properties {
 maxDistance=2097152
 soundClassObject=&Fort_UI_InGame_Critical
 volumeMultiplier=1
 concurrencySet [&/CS_VoidMask]
}
