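# World marker indicator widget for the Demon Mask Void teleport marker, built on AthenaSquadIndicator.
# Marker data binds it to the teleport actor and switches visuals and offset for owner, ally (pin-colored diamond) or enemy, with red coloring for opposing teams.
# The intro scale animation plays immediately, or after a short hidden delay for the owner.
# For the owner, leash events drive a looping timer that shows an alert glow and plays a one-shot warning near expiry.
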WidgetBlueprint /DemonMaskVoid/UI/TeleportMarker/WBP_DemonMaskVoid_MarkedActorIndicator
animations [Anim_Intro {
  binding_count=1
  length_seconds=0.7
  object=&self:Anim_Intro
  sequence {bindings [Overlay_Root {
     id=#1
     possessed_class=Overlay
     tracks [{
       type=MovieScene2DTransformTrack
       sections [{
         type=MovieScene2DTransformSection
         end_frame=14
         start_frame=0
         channels [
          Scale.X {type=MovieSceneScriptingFloatChannel, keys [[0,1],[11,1.5],[14,1]]}
          Scale.Y {type=MovieSceneScriptingFloatChannel, keys [[0,1],[11,1.5],[14,1]]}
         ]
        }]
      }]
    }]}
 }]
graphs [
 BindTurretEvents {nodes [
   n1 = BindTurretEvents!() -> (then, InTeleportMarkerActor:BGA Demon Mask Void Teleport Base Object Reference) {is=title K2Node_FunctionEntry}
   > UnbindTurretEvents()
   > n2 = Set TeleportMarkerActor(TeleportMarkerActor:BGA Demon Mask Void Teleport Base Object Reference=@n1) -> (Output_Get:BGA Demon Mask Void Teleport Base Object Reference)
   > Is Valid(InputObject=@n2)
   > Bind Event to On Leash Timer Active Updated Event Dispatcher(self:BGA Demon Mask Void Teleport Base Object Reference=@n1, Delegate=Create Event()) {is=title K2Node_AddDelegate}
   > OnLeashActiveUpdated!(
    execute
    self
    IsLeashActive:bool=Get IsLeashTimerActive!(self:BGA Demon Mask Void Teleport Base Object Reference=@n1) -> (IsLeashTimerActive:bool)
   ) -> (then)
  ]}
 EventGraph {nodes [
   n1 = OnAnimationWaitFinished() {is=title K2Node_CustomEvent}
   PlayAnimation(execute=[@n1.then, @n7.else], InAnimation=Get Anim_Intro() -> (Anim_Intro:Widget Animation Object Reference))
   > Clear and Invalidate Timer by Handle(Handle=Get AnimationWaitTimerHandle() -> (AnimationWaitTimerHandle:Timer Handle Structure))
   > SetVisibility(self=Get Overlay_Root() -> (Overlay_Root:Overlay Object Reference), InVisibility=SelfHitTestInvisible)
   n2 = Event OnItemIconSet!() -> (OutputDelegate, then, NewIcon:Texture 2D Object Reference) {is=title K2Node_Event}
   > n3 = GetDynamicMaterial(self=Get Image_OwnerIndicator_Small() -> (Image_OwnerIndicator_Small:Image Object Reference))
   > SetTextureParameterValue(self=@n3, ParameterName=DisplayedIcon, Value=@n2.NewIcon)
   n4 = Event OnSetMarkerData!() -> (OutputDelegate, then, FortWorldMarkerData:"Fort World Marker Data Structure (by ref)") {is=title K2Node_Event}
   > n5 = Sequence()
   > n6 = Cast To BGA_DemonMask_Void_Teleport_Base!(
    execute
    Object=Break Fort World Marker Data!(FortWorldMarkerData:"Fort World Marker Data Structure (by ref)"=@n4.FortWorldMarkerData) -> (
     MarkerID:Marker ID Structure
     Owner:Fort Player State Weak object
     MarkerType:EFortWorldMarkerType Enum
     BasePositionOffset:Vector=~
     WorldNormal:Vector=~
     ItemDefinition:Item Definition Base Object Reference
     ItemCount:int
     PinnedQuestIndex:int
     MarkedActor:Actor Weak object
     MarkedActorClass:Actor Soft Class Reference
     MarkerTag:Gameplay Tag Structure
     bUseHoveredMarkerDetail:bool=~
    ) {is=title K2Node_BreakStruct}.MarkedActor
   ) -> (then, CastFailed, AsBGA Demon Mask Void Teleport Base:BGA Demon Mask Void Teleport Base Object Reference) {is=title K2Node_DynamicCast}
   > BindTurretEvents!(execute, self, InTeleportMarkerActor:BGA Demon Mask Void Teleport Base Object Reference=@n6) -> (then) {is=title K2Node_CallFunction}
   > UpdateOwnership()
   n7 = Branch(execute=@n5.then_1, Condition=IsOwner!(self) -> (IsOwner:bool))
   > n8 = Set Timer by Event(Delegate=@n1, Time=1, bLooping=false)
   > Set AnimationWaitTimerHandle(AnimationWaitTimerHandle:Timer Handle Structure=@n8) -> (Output_Get:Timer Handle Structure)
   > SetVisibility(self=Get Overlay_Root(), InVisibility=Collapsed)
  ]}
 IsOwner {nodes [
   IsOwner() {is=title K2Node_FunctionEntry}
   n1 = Get TeleportMarkerActor() -> (TeleportMarkerActor:BGA Demon Mask Void Teleport Base Object Reference)
   > n2 = Is Valid(InputObject=@n1)
   Return Node!(execute=@n2.Is Not Valid, IsOwner:bool=false)
   > Return Node(IsOwner="Equal (Object)"(
     A=GetOwner(self:Actor Object Reference=@n1)
     B=GetOwningPlayerPawn!(self:User Widget Object Reference) -> (ReturnValue:Pawn Object Reference)
    ))
  ]}
 OnLeashActiveUpdated {nodes [
   n1 = OnLeashActiveUpdated!() -> (then, IsLeashActive:bool=~) {is=title K2Node_FunctionEntry}
   > Branch(Condition=IsOwner() {is=title K2Node_CallFunction})
   > n2 = Branch(Condition=@n1)
   > n3 = Set Timer by Event(Delegate=Create Event(), Time=1, bLooping=true)
   > Set LeashTimerHandle(LeashTimerHandle:Timer Handle Structure=@n3) -> (Output_Get:Timer Handle Structure)
   > UpdateTimeRemaining()
   Clear and Invalidate Timer by Handle(execute=@n2.else, Handle=Get LeashTimerHandle() -> (LeashTimerHandle:Timer Handle Structure))
   > SetVisibility(self=Get Image_AlertGlow() -> (Image_AlertGlow:Image Object Reference), InVisibility=Collapsed)
   > Stop(self:Audio Component Object Reference=Get ActiveWarningSound() -> (ActiveWarningSound:Audio Component Object Reference))
  ]}
 UnbindTurretEvents {nodes [
   UnbindTurretEvents() {is=title K2Node_FunctionEntry}
   n1 = Get TeleportMarkerActor()
   > Is Valid(InputObject=@n1)
   > Unbind Event from On Leash Timer Active Updated Event Dispatcher(self:BGA Demon Mask Void Teleport Base Object Reference=@n1, Delegate=Create Event()) {is=title K2Node_RemoveDelegate}
   > Set TeleportMarkerActor()
  ]}
 UpdateIconColor {nodes [
   n1 = UpdateIconColor!() -> (then, IconWidget:Image Object Reference, Color:Linear Color Structure) {is=title K2Node_FunctionEntry}
   > n2 = GetDynamicMaterial(self=@n1)
   > SetScalarParameterValue(self=@n2, ParameterName=UseIconColor, Value=1)
   > SetScalarParameterValue(self=@n2, ParameterName=IconColorStrength, Value=0.94)
   n3 = BreakColor(InColor=@n1.Color)
   n4 = Get ColorReduction() -> (ColorReduction:double)
   > SetVectorParameterValue(
    self=@n2
    ParameterName=FillColor
    Value=MakeColor!(R:float=float * float(A=@n3, B=@n4), G:float=float * float(A=@n3.G, B=@n4), B:float=float * float(A=@n3.B, B=@n4), A:float=0.9) -> (ReturnValue:Linear Color Structure)
   )
   > SetVectorParameterValue(self=@n2, ParameterName=FillColorInner, Value=@n1.Color)
   > SetVectorParameterValue(self=@n2, ParameterName=RingColor, Value=@n1.Color)
   > SetVectorParameterValue(self=@n2, ParameterName=IconColor, Value=(R=1,G=1,B=1,A=1))
   > SetVectorParameterValue(self=@n2, ParameterName=IconScale, Value=MakeColor(R=0.85, G=0.85, B=0, A=1))
  ]}
 UpdateLargeIndicatorColor {nodes [
   UpdateLargeIndicatorColor!() -> (then, DefaultColor:Linear Color Structure) {is=title K2Node_FunctionEntry}
   n1 = Get TeleportMarkerActor()
   > n2 = Is Valid(InputObject=@n1)
   > n3 = Branch(Condition=OnSameTeam!(ActorA:Actor Object Reference=@n1, ActorB:Actor Object Reference=GetOwningPlayerPawn()) -> (ReturnValue:bool))
   UpdateIconColor!(
    execute=[@n3.then, @n2.Is Not Valid]
    self
    IconWidget:Image Object Reference=Get Image_LargeIndicator() -> (Image_LargeIndicator:Image Object Reference)
    Color:Linear Color Structure=Get DefaultColor() -> (DefaultColor:Linear Color Structure)
   ) -> (then) {is=title K2Node_CallFunction}
   > UpdateIconColor(IconWidget=Get Image_OwnerIndicator_Small(), Color=Get DefaultColor())
   UpdateIconColor(execute=@n3.else, IconWidget=Get Image_LargeIndicator(), Color=(R=1,G=0,B=0,A=1))
   > SetColorAndOpacity(
    self:Image Object Reference=Get Image_SmallIndicator() -> (Image_SmallIndicator:Image Object Reference)
    InColorAndOpacity:Linear Color Structure=(R=1,G=0,B=0,A=1)
   )
  ]}
 UpdateOwnership {nodes [
   UpdateOwnership() {is=title K2Node_FunctionEntry}
   > n1 = Branch(Condition=IsOwner())
   > SetActiveWidgetIndex(
    self:Common Visibility Switcher Object Reference=[
     Get Switcher_SmallIndicator() -> (Switcher_SmallIndicator:Common Visibility Switcher Object Reference)
     Get Switcher_LargeIndicator() -> (Switcher_LargeIndicator:Common Visibility Switcher Object Reference)
    ]
    Index=0
   )
   > SetVisibility(self=Get Image_AllyIndicator_Large() -> (Image_AllyIndicator_Large:Image Object Reference), InVisibility=Collapsed)
   > Set ScreenSpaceRelativeOffset(ScreenSpaceRelativeOffset:Vector 2D Structure=Get AllyWidgetOffset() -> (AllyWidgetOffset:Vector 2D Structure)) -> (Output_Get:Vector 2D Structure)
   n2 = Get TeleportMarkerActor()
   n3 = Branch(execute=@n1.else, Condition=OnSameTeam(ActorA=GetOwner(self=@n2), ActorB=GetOwningPlayer()))
   > SetActiveWidgetIndex(self=Get Switcher_SmallIndicator(), Index=1)
   > SetActiveWidgetIndex(self=Get Switcher_LargeIndicator(), Index=0)
   > SetVisibility(self=Get Image_AllyIndicator_Large(), InVisibility=SelfHitTestInvisible)
   > Set ScreenSpaceRelativeOffset(ScreenSpaceRelativeOffset=Get AllyWidgetOffset())
   > n4 = Cast To FortPlayerPawn(Object=GetOwner(self=Get TeleportMarkerActor()))
   > n5 = AttemptGetPlayerStateForPawn!(execute, FortPawn:Fort Player Pawn Object Reference=@n4) -> (then, ReturnValue:Fort Player State Object Reference)
   > n6 = Cast To FortPlayerStateAthena!(execute, Object=@n5) -> (then, CastFailed, AsFort Player State Athena:Fort Player State Athena Object Reference) {is=title K2Node_DynamicCast}
   > SetColorAndOpacity(
    self=Get Image_AllyIndicator_Large()
    InColorAndOpacity=GetPinColor!(self:Fort Player State Zone Object Reference=@n6) -> (ReturnValue:Linear Color Structure)
   )
   SetActiveWidgetIndex(execute=@n3.else, self=Get Switcher_SmallIndicator(), Index=2)
   > SetActiveWidgetIndex(self=Get Switcher_LargeIndicator(), Index=1)
   > Set ScreenSpaceRelativeOffset(ScreenSpaceRelativeOffset=Get EnemyWidgetOffset() -> (EnemyWidgetOffset:Vector 2D Structure))
  ]}
 UpdateTimeRemaining {nodes [
   UpdateTimeRemaining() {is=title K2Node_FunctionEntry}
   > n1 = Is Valid(InputObject=Get TeleportMarkerActor())
   OnLeashActiveUpdated(execute=@n1.Is Not Valid, IsLeashActive=false) {is=title K2Node_CallFunction}
   Set TimeRemaining(
    execute=@n1.Is Valid
    TimeRemaining:float=Get CurrentLeashTimeRemaining!(self:BGA Demon Mask Void Teleport Base Object Reference=Get TeleportMarkerActor()) -> (CurrentLeashTimeRemaining:float)
   ) -> (Output_Get:float)
   n2 = "float < float"(A=Get TimeRemaining() -> (TimeRemaining:float), B=5)
   > SetVisibility(self=Get Image_AlertGlow(), InVisibility=Select(Option 0=Collapsed, Option 1=SelfHitTestInvisible, Index=@n2))
   > n3 = Branch(Condition=AND Boolean(A=@n2, B=Get bPlayOnce() -> (bPlayOnce:bool)))
   > n4 = SpawnSoundAtLocation!(
    execute
    Sound:Sound Base Object Reference=Get WarningSoundCue() -> (WarningSoundCue:Sound Base Object Reference)
    Location:Vector=Get Actor Location(self=Get TeleportMarkerActor())
    Rotation:Rotator
    VolumeMultiplier:float=1
    PitchMultiplier:float=1
    StartTime:float=0
    AttenuationSettings:Sound Attenuation Object Reference
    ConcurrencySettings:Sound Concurrency Object Reference
    bAutoDestroy:bool=true
   ) -> (then, ReturnValue:Audio Component Object Reference)
   > Set ActiveWarningSound(ActiveWarningSound:Audio Component Object Reference=@n4) -> (Output_Get:Audio Component Object Reference)
   > Set bPlayOnce(bPlayOnce:bool=false) -> (Output_Get:bool)
   Branch(execute=@n3.else, Condition="float > float"(A=Get TimeRemaining(), B=5))
   > Set bPlayOnce(bPlayOnce=true)
  ]}
]
variables [
 ActiveWarningSound {type=AudioComponent, default=null}
 AllyWidgetOffset {type=Vector2D, default=<0 -50>}
 AnimationWaitTimerHandle {type=TimerHandle}
 bPlayOnce {type=boolean, default=true}
 ColorReduction {type=number, default=0.08}
 EnemyWidgetOffset {type=Vector2D, default=<0 -25>}
 LeashTimerHandle {type=TimerHandle}
 OwningPlayerST {type=FortPlayerState, default=null}
 TeleportMarkerActor {type=BGA_DemonMask_Void_Teleport_Base_C, default=null}
 WarningSoundCue {type=SoundBase, default=/DemonMaskVoid/Assets/Audio/SC_VoidMask_UI_Marker_ExpireWarning}
]
widgets {
 info {inheritedWidgetCount=11, parentClass=/Script/FortniteUI.AthenaSquadIndicator, rootWidgetClass=&Overlay, widgetCount=25}
 widgets [
  Overlay_Root {widgetClassPath=&Overlay, bIsVariable=true, widget=Overlay_Root}
  Spacer_SizeConsistency {widgetClassPath=&Spacer, parent=Overlay_Root, widget=Spacer_SizeConsistency, properties {size=<72 72>}, slot {class=OverlaySlot}}
  Switcher_IndicatorSize {
   widgetClassPath=&CommonVisibilitySwitcher
   bInherited=true
   parent=Overlay_Root
   widget=Switcher_IndicatorSize
   slot {class=OverlaySlot, properties {horizontalAlignment=HAlign_Fill, verticalAlignment=VAlign_Fill}}
  }
  Switcher_SmallIndicator {
   widgetClassPath=&CommonVisibilitySwitcher
   bIsVariable=true
   parent=Switcher_IndicatorSize
   widget=Switcher_SmallIndicator
   slot {class=CommonVisibilitySwitcherSlot}
  }
  Image_OwnerIndicator_Small {
   widgetClassPath=&Image
   bIsVariable=true
   parent=Switcher_SmallIndicator
   widget=Image_OwnerIndicator_Small
   properties {brush {
     bIsDynamicallyLoaded=false
     drawAs=Image
     imageType=FullColor
     mirroring=NoMirror
     resourceName=None
     resourceObject=/DemonMaskVoid/UI/Materials/M_UI_PingBubble_LargeItemPing
     tiling=NoTile
     imageSize=<72 72>
     margin=<ltrb 0 0 0 0>
     outlineSettings {
      bUseBrushTransparency=false
      roundingType=HalfHeightRadius
      width=0
      color {colorUseRule=UseColor_Specified, specifiedColor=<rgba 0 0 0 0>}
      cornerRadii=<0 0 0 0>
     }
     tintColor {colorUseRule=UseColor_Specified, specifiedColor=<rgba 1 1 1 1>}
     uVRegion {bIsValid=false, max=<0 0>, min=<0 0>}
    }}
   slot {class=CommonVisibilitySwitcherSlot, properties {horizontalAlignment=HAlign_Center, verticalAlignment=VAlign_Center}}
  }
  Spacer_EmptyAllyIndicator {
   widgetClassPath=&Spacer
   parent=Switcher_SmallIndicator
   widget=Spacer_EmptyAllyIndicator
   slot {class=CommonVisibilitySwitcherSlot}
  }
  Image_SmallIndicator {
   widgetClassPath=&Image
   bInherited=true
   bIsVariable=true
   parent=Switcher_SmallIndicator
   widget=Image_SmallIndicator
   properties {
    brush {
     bIsDynamicallyLoaded=false
     drawAs=Image
     imageType=NoImage
     mirroring=NoMirror
     resourceName=None
     resourceObject=/Game/Athena/HUD/Materials/M_PingIndicator
     tiling=NoTile
     imageSize=<16 16>
     margin=<ltrb 0 0 0 0>
     outlineSettings {
      bUseBrushTransparency=false
      roundingType=HalfHeightRadius
      width=0
      color {colorUseRule=UseColor_Specified, specifiedColor=<rgba 0 0 0 0>}
      cornerRadii=<0 0 0 0>
     }
     tintColor {colorUseRule=UseColor_Specified, specifiedColor=<rgba 1 1 1 1>}
     uVRegion {bIsValid=false, max=<0 0>, min=<0 0>}
    }
    renderTransform {angle=0, scale=<1 1>, shear=<0 0>, translation=<0 -3>}
   }
   slot {class=CommonVisibilitySwitcherSlot, properties {horizontalAlignment=HAlign_Center, verticalAlignment=VAlign_Center}}
  }
  Overlay_LargeIndicator {widgetClassPath=&Overlay, parent=Switcher_IndicatorSize, widget=Overlay_LargeIndicator, slot {class=CommonVisibilitySwitcherSlot}}
  Image_LocationFlare {
   widgetClassPath=&Image
   bInherited=true
   parent=Overlay_LargeIndicator
   widget=Image_LocationFlare
   properties {
    visibility=SelfHitTestInvisible
    brush {
     bIsDynamicallyLoaded=false
     drawAs=Image
     imageType=NoImage
     mirroring=NoMirror
     resourceName=None
     resourceObject=/Game/Athena/HUD/Materials/M_UI_LocationPingReminder
     tiling=NoTile
     imageSize=<25 25>
     margin=<ltrb 0 0 0 0>
     outlineSettings {
      bUseBrushTransparency=false
      roundingType=HalfHeightRadius
      width=0
      color {colorUseRule=UseColor_Specified, specifiedColor=<rgba 0 0 0 0>}
      cornerRadii=<0 0 0 0>
     }
     tintColor {colorUseRule=UseColor_Specified, specifiedColor=<rgba 1 1 1 1>}
     uVRegion {bIsValid=false, max=<0 0>, min=<0 0>}
    }
    renderTransform {angle=0, scale=<3 3>, shear=<0 0>, translation=<0 6>}
   }
   slot {class=OverlaySlot, properties {horizontalAlignment=HAlign_Center, verticalAlignment=VAlign_Center}}
  }
  CommonVisualAttachment_0 {
   widgetClassPath=&CommonVisualAttachment
   parent=Overlay_LargeIndicator
   widget=CommonVisualAttachment_0
   properties {contentAnchor=<0.5 0.5>}
   slot {class=OverlaySlot, properties {horizontalAlignment=HAlign_Center, verticalAlignment=VAlign_Center}}
  }
  Image_AlertGlow {
   widgetClassPath=&Image
   bIsVariable=true
   parent=CommonVisualAttachment_0
   widget=Image_AlertGlow
   properties {
    visibility=Collapsed
    brush {
     bIsDynamicallyLoaded=false
     drawAs=Image
     imageType=NoImage
     mirroring=NoMirror
     resourceName=None
     resourceObject=/Game/Athena/HUD/LTM/Wax/Materials/M_UI_TextGlow_Add_Gradient_Orange_Nevada
     tiling=NoTile
     imageSize=<72 72>
     margin=<ltrb 0 0 0 0>
     outlineSettings {
      bUseBrushTransparency=false
      roundingType=HalfHeightRadius
      width=0
      color {colorUseRule=UseColor_Specified, specifiedColor=<rgba 0 0 0 0>}
      cornerRadii=<0 0 0 0>
     }
     tintColor {colorUseRule=UseColor_Specified, specifiedColor=<rgba 1 1 1 1>}
     uVRegion {bIsValid=false, max=<0 0>, min=<0 0>}
    }
    renderTransform {angle=0, scale=<2 2>, shear=<0 0>, translation=<0 0>}
   }
   slot {class=SizeBoxSlot}
  }
  Switcher_LargeIndicator {
   widgetClassPath=&CommonVisibilitySwitcher
   bIsVariable=true
   parent=Overlay_LargeIndicator
   widget=Switcher_LargeIndicator
   slot {class=OverlaySlot}
  }
  HorizontalBox_0 {
   widgetClassPath=&HorizontalBox
   parent=Switcher_LargeIndicator
   widget=HorizontalBox_0
   slot {class=CommonVisibilitySwitcherSlot, properties {horizontalAlignment=HAlign_Center, verticalAlignment=VAlign_Center}}
  }
  Image_LargeIndicator {
   widgetClassPath=&Image
   bInherited=true
   bIsVariable=true
   parent=HorizontalBox_0
   widget=Image_LargeIndicator
   properties {brush {
     bIsDynamicallyLoaded=false
     drawAs=Image
     imageType=FullColor
     mirroring=NoMirror
     resourceName=None
     resourceObject=&M_UI_PingBubble_LargeItemPing
     tiling=NoTile
     imageSize=<72 72>
     margin=<ltrb 0 0 0 0>
     outlineSettings {
      bUseBrushTransparency=false
      roundingType=HalfHeightRadius
      width=0
      color {colorUseRule=UseColor_Specified, specifiedColor=<rgba 0 0 0 0>}
      cornerRadii=<0 0 0 0>
     }
     tintColor {colorUseRule=UseColor_Specified, specifiedColor=<rgba 1 1 1 1>}
     uVRegion {bIsValid=false, max=<0 0>, min=<0 0>}
    }}
   slot {class=HorizontalBoxSlot}
  }
  CommonVisualAttachment_1 {
   widgetClassPath=&CommonVisualAttachment
   parent=HorizontalBox_0
   widget=CommonVisualAttachment_1
   properties {contentAnchor=<0 1>}
   slot {class=HorizontalBoxSlot, properties {horizontalAlignment=HAlign_Center, verticalAlignment=VAlign_Center}}
  }
  Image_AllyIndicator_Large {
   widgetClassPath=&Image
   bIsVariable=true
   parent=CommonVisualAttachment_1
   widget=Image_AllyIndicator_Large
   properties {
    visibility=SelfHitTestInvisible
    brush {
     bIsDynamicallyLoaded=false
     drawAs=Image
     imageType=NoImage
     mirroring=NoMirror
     resourceName=None
     resourceObject=/Game/Athena/HUD/Art/T_UI_Indicator_Diamond
     tiling=NoTile
     imageSize=<28 32>
     margin=<ltrb 0 0 0 0>
     outlineSettings {
      bUseBrushTransparency=false
      roundingType=HalfHeightRadius
      width=0
      color {colorUseRule=UseColor_Specified, specifiedColor=<rgba 0 0 0 0>}
      cornerRadii=<0 0 0 0>
     }
     tintColor {colorUseRule=UseColor_Specified, specifiedColor=<rgba 1 1 1 1>}
     uVRegion {bIsValid=false, max=<0 0>, min=<0 0>}
    }
   }
   slot {class=SizeBoxSlot}
  }
  Image_DangerIcon {
   widgetClassPath=&Image
   bIsVariable=true
   parent=Switcher_LargeIndicator
   widget=Image_DangerIcon
   properties {
    brush {
     bIsDynamicallyLoaded=false
     drawAs=Image
     imageType=NoImage
     mirroring=NoMirror
     resourceName=None
     resourceObject=/Game/Athena/HUD/Materials/M_UI_PingBubble_Inst
     tiling=NoTile
     imageSize=<72 72>
     margin=<ltrb 0 0 0 0>
     outlineSettings {
      bUseBrushTransparency=false
      roundingType=HalfHeightRadius
      width=0
      color {colorUseRule=UseColor_Specified, specifiedColor=<rgba 0 0 0 0>}
      cornerRadii=<0 0 0 0>
     }
     tintColor {colorUseRule=UseColor_Specified, specifiedColor=<rgba 1 1 1 1>}
     uVRegion {bIsValid=false, max=<0 0>, min=<0 0>}
    }
    renderTransformPivot=<1 0.5>
   }
   slot {class=CommonVisibilitySwitcherSlot, properties {horizontalAlignment=HAlign_Center, verticalAlignment=VAlign_Center}}
  }
  VisAttachment_Distance {
   widgetClassPath=&FortVisualAttachment
   parent=Overlay_Root
   widget=VisAttachment_Distance
   properties {renderTransform {angle=0, scale=<1 1>, shear=<0 0>, translation=<-10 0>}}
   slot {class=OverlaySlot, properties {horizontalAlignment=HAlign_Right, verticalAlignment=VAlign_Center}}
  }
  Text_Distance {
   widgetClassPath=&CommonTextBlock
   bInherited=true
   parent=VisAttachment_Distance
   widget=Text_Distance
   properties {
    text=100m
    font {
     bForceMonospaced=false
     bMaterialIsStencil=false
     fontMaterial=null
     fontObject=&Burbank
     letterSpacing=0
     monospacedWidth=1
     size=14
     skewAmount=0.2
     typefaceFontName=BurbankRegular-Black
     outlineSettings {
      bApplyOutlineToDropShadows=false
      bMiteredCorners=false
      bSeparateFillAlpha=false
      outlineMaterial=null
      outlineSize=1
      outlineColor=<rgba 0 0 0 0.6>
     }
    }
    shadowColorAndOpacity=<rgba 0 0 0 1>
    shadowOffset=<3 3>
   }
   slot {class=SizeBoxSlot}
  }
  Image_TrajectoryLargeIndicator {widgetClassPath=&Image, bInherited=true, widget=null}
  Image_TrajectorySmallIndicator {widgetClassPath=&Image, bInherited=true, widget=null}
  MarkerConfirmationWidget {widgetClassPath=/Script/FortniteUI.AthenaMarkerConfirmationWidget, bInherited=true, widget=null}
  Panel_SmallIndicator {widgetClassPath=/Script/UMG.PanelWidget, bInherited=true, widget=null}
  Panel_LargeIndicator {widgetClassPath=&PanelWidget, bInherited=true, widget=null}
  TutorialHighlight {widgetClassPath=&TutorialHighlight}
 ]
}
