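# Marker details widget for a pinged Demon Mask Void teleport, showing its name, icon and a cancel-marker keybind prompt.
# SetMarkerData casts the marked actor and compares its owner's team with the viewing player's pawn.
# Teammates get the Epic rarity color and item icon, others red and a danger icon.
# UpdateVisuals tints the separator, the icon's dynamic material and the line of the parent AthenaMarkerDetails found on Construct.
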WidgetBlueprint /DemonMaskVoid/UI/TeleportMarker/WBP_DemonMaskVoid_MarkerDetails
graphs [
 EventGraph {nodes [
   n1 = Event SetMarkerData!() -> (OutputDelegate, then, MarkerData:"Fort World Marker Data Structure (by ref)", DisplayInfo:"Marked Actor Display Info Structure (by ref)") {is=title K2Node_Event}
   n2 = Break Fort World Marker Data!(FortWorldMarkerData:"Fort World Marker Data Structure (by ref)"=@n1.MarkerData) -> (
    MarkerID:Marker ID Structure
    Owner:Fort Player State Weak object
    MarkerType:EFortWorldMarkerType Enum
    BasePositionOffset:Vector=~
    WorldNormal:Vector=~
    ItemDefinition:Item Definition Base Object Reference
    ItemCount:int
    PinnedQuestIndex:int
    MarkedActor:Actor Weak object
    MarkedActorClass:Actor Soft Class Reference
    MarkerTag:Gameplay Tag Structure
    bUseHoveredMarkerDetail:bool=~
   ) {is=title K2Node_BreakStruct}
   > n3 = Cast To BGA_DemonMask_Void_Teleport_Base!(execute, Object=@n2.MarkedActor) -> (then, CastFailed, AsBGA Demon Mask Void Teleport Base:BGA Demon Mask Void Teleport Base Object Reference) {is=title K2Node_DynamicCast}
   > n4 = GetRarityData!(execute, Rarity:EFortRarity Enum=Epic) -> (then, ReturnValue:Fort Rarity Item Data Structure)
   n5 = OnSameTeam!(
    ActorA:Actor Object Reference=GetOwner(self:Actor Object Reference=@n3)
    ActorB:Actor Object Reference=GetOwningPlayerPawn!(self:User Widget Object Reference) -> (ReturnValue:Pawn Object Reference)
   ) -> (ReturnValue:bool)
   n6 = Break Marked Actor Display Info!(MarkedActorDisplayInfo:"Marked Actor Display Info Structure (by ref)"=@n1.DisplayInfo) -> (
    DisplayName:Text
    Icon:Texture 2D Soft Object Reference
    CustomIndicatorClass:User Widget Soft Class Reference
    CustomDetailsClass:User Widget Soft Class Reference
    PrimaryColor:Linear Color Structure
    Sound:Sound Base Object Reference
    bAttachSound:bool=~
    ScreenClamping:EFortMarkedActorScreenClamping Enum
    SpecialCull:Bitmask Integer
    SpecialDotSizeThreshold:float=~
    SpecialDistance:float=~
    CustomTimeToLive:float=~
    AdditionalData:Marked Actor Display Info Additional Data Object Reference
   ) {is=title K2Node_BreakStruct}
   > UpdateVisuals!(
    execute
    self
    Color:Linear Color Structure=Select(
     Option 0:Linear Color Structure=(R=1,G=0,B=0,A=1)
     Option 1:Linear Color Structure=Break Fort Rarity Item Data!(FortRarityItemData:"Fort Rarity Item Data Structure (by ref)"=@n4) -> (
      Name:Text
      Color1:Linear Color Structure
      Color2:Linear Color Structure
      Color3:Linear Color Structure
      Color4:Linear Color Structure
      Color5:Linear Color Structure
      Radius:float=~
      Falloff:float=~
      Brightness:float=~
      Roughness:float=~
      Glow:float=~
     ) {is=title K2Node_BreakStruct}.Color2
     Index=@n5
    ) -> (ReturnValue:Linear Color Structure)
    Icon:Texture 2D Soft Object Reference=@n6.Icon
   ) -> (then)
   > SetActiveWidgetIndex(
    self:Common Visibility Switcher Object Reference=Get Switcher_ItemIcon() -> (Switcher_ItemIcon:Common Visibility Switcher Object Reference)
    Index=SelectInt(A=0, B=1, bPickA=@n5)
   )
   > "SetText (Text)"(self=Get ItemName() -> (ItemName:Text Object Reference), InText=@n6)
   Event Construct()
   > Parent: Construct()
   > n7 = FindParentWidgetOfType!(
    execute
    StartingWidget:Widget Object Reference=GetParent!(self:Widget Object Reference) -> (ReturnValue:Panel Widget Object Reference)
    Type:Widget Class Reference=/Game/Athena/HUD/Team/AthenaMarkerDetails.AthenaMarkerDetails_C
   ) -> (then, ReturnValue:Athena Marker Details Object Reference)
   > Set AthenaMarkerDetailsParent(AthenaMarkerDetailsParent:Athena Marker Details Object Reference=@n7) -> (Output_Get:Athena Marker Details Object Reference)
   > "SetText (Text)"(self=Get ActionText() -> (ActionText:Common Text Object Reference), InText="LOCTEXT(\"Cancel Marker\")")
   Reroute Node(InputPin:Actor Weak object=@n2.MarkedActor) -> (OutputPin:Actor Weak object)
  ]}
 UpdateIcon {nodes [
   n1 = UpdateIcon!() -> (then, Icon:Texture 2D Soft Object Reference, Color:Linear Color Structure) {is=title K2Node_FunctionEntry}
   > n2 = GetDynamicMaterial(self=Get Image_ItemIcon() -> (Image_ItemIcon:Image Object Reference))
   > SetScalarParameterValue(self=@n2, ParameterName=UseIconColor, Value=1)
   > SetScalarParameterValue(self=@n2, ParameterName=IconColorStrength, Value=0.94)
   > SetTextureParameterValue(
    self=@n2
    ParameterName=DisplayedIcon
    Value=Resolve Soft Reference!(Input:Texture 2D Soft Object Reference=@n1) -> (Output:Texture 2D Object Reference)
   )
   n3 = BreakColor(InColor=@n1.Color)
   n4 = Get ColorReduction() -> (ColorReduction:double)
   > SetVectorParameterValue(
    self=@n2
    ParameterName=FillColor
    Value=MakeColor!(R:float=float * float(A=@n3, B=@n4), G:float=float * float(A=@n3.G, B=@n4), B:float=float * float(A=@n3.B, B=@n4), A:float=0.9) -> (ReturnValue:Linear Color Structure)
   )
   > SetVectorParameterValue(self=@n2, ParameterName=FillColorInner, Value=@n1.Color)
   > SetVectorParameterValue(self=@n2, ParameterName=RingColor, Value=@n1.Color)
   > SetVectorParameterValue(self=@n2, ParameterName=IconColor, Value=(R=1,G=1,B=1,A=1))
   > SetVectorParameterValue(self=@n2, ParameterName=IconScale, Value=MakeColor(R=0.85, G=0.85, B=0, A=1))
  ]}
 UpdateVisuals {nodes [
   n1 = UpdateVisuals!() -> (then, Color:Linear Color Structure, Icon:Texture 2D Soft Object Reference) {is=title K2Node_FunctionEntry}
   > SetBrushColor(self=Get Seperator() -> (Seperator:Border Object Reference), InBrushColor=@n1)
   > UpdateIcon!(execute, self, Icon:Texture 2D Soft Object Reference=@n1.Icon, Color:Linear Color Structure=@n1) -> (then) {is=title K2Node_CallFunction}
   n2 = Get AthenaMarkerDetailsParent() -> (AthenaMarkerDetailsParent:Athena Marker Details Object Reference)
   > Is Valid(InputObject=@n2)
   > SetLineColor!(execute, self:Athena Marker Details Object Reference=@n2, NewColor:Linear Color Structure=@n1) -> (then)
  ]}
]
variables [
 AthenaMarkerDetailsParent {type=AthenaMarkerDetails_C, default=null}
 ColorReduction {type=number, default=0.08}
]
widgets {
 info {inheritedWidgetCount=0, parentClass=&UserWidget, rootWidgetClass=&VerticalBox, widgetCount=12}
 widgets [
  VerticalBox_0 {widgetClassPath=&VerticalBox, widget=VerticalBox_0}
  ItemName {
   widgetClassPath=&TextBlock
   bIsVariable=true
   parent=VerticalBox_0
   widget=ItemName
   properties {
    text="{MARKER TITLE}"
    font {
     bForceMonospaced=false
     bMaterialIsStencil=false
     fontMaterial=null
     fontObject=&Burbank
     letterSpacing=0
     monospacedWidth=1
     size=20
     skewAmount=0
     typefaceFontName=Burbank-PlayerPower
     outlineSettings {
      bApplyOutlineToDropShadows=true
      bMiteredCorners=false
      bSeparateFillAlpha=true
      outlineMaterial=&M-FontShadow
      outlineSize=3
      outlineColor=<rgba 0 0 0 0.8>
     }
    }
    renderTransform {angle=0, scale=<1 1>, shear=<-8 0>, translation=<0 0>}
   }
   slot {class=VerticalBoxSlot, properties {horizontalAlignment=HAlign_Left, padding=<ltrb 6 0 8 -6>}}
  }
  Overlay_0 {
   widgetClassPath=&Overlay
   parent=VerticalBox_0
   widget=Overlay_0
   slot {class=VerticalBoxSlot, properties {padding=<ltrb 0 4 0 0>}}
  }
  Seperator {
   widgetClassPath=&Border
   bIsVariable=true
   parent=Overlay_0
   widget=Seperator
   properties {
    background {
     bIsDynamicallyLoaded=false
     drawAs=Image
     imageType=NoImage
     mirroring=NoMirror
     resourceName=None
     resourceObject=null
     tiling=NoTile
     imageSize=<32 16>
     margin=<ltrb 0 0 0 0>
     outlineSettings {
      bUseBrushTransparency=false
      roundingType=HalfHeightRadius
      width=0
      color {colorUseRule=UseColor_Specified, specifiedColor=<rgba 0 0 0 0>}
      cornerRadii=<0 0 0 0>
     }
     tintColor {colorUseRule=UseColor_Specified, specifiedColor=<rgba 1 1 1 1>}
     uVRegion {bIsValid=false, max=<0 0>, min=<0 0>}
    }
    brushColor=<rgba 1 0 0 1>
    padding=<ltrb 4 0 4 3>
    renderTransformPivot=<0 0.5>
   }
   slot {class=OverlaySlot, properties {horizontalAlignment=HAlign_Fill, verticalAlignment=VAlign_Center, padding=<ltrb 3 0 0 0>}}
  }
  FortVisualAttachment_0 {
   widgetClassPath=&FortVisualAttachment
   parent=Overlay_0
   widget=FortVisualAttachment_0
   properties {widthOverride=100}
   slot {class=OverlaySlot, properties {verticalAlignment=VAlign_Center}}
  }
  Switcher_ItemIcon {
   widgetClassPath=&CommonVisibilitySwitcher
   bIsVariable=true
   parent=FortVisualAttachment_0
   widget=Switcher_ItemIcon
   properties {renderTransform {angle=0, scale=<1 1>, shear=<0 0>, translation=<-65 -37>}}
   slot {class=SizeBoxSlot, properties {horizontalAlignment=HAlign_Left, verticalAlignment=VAlign_Top}}
  }
  Image_ItemIcon {
   widgetClassPath=&Image
   bIsVariable=true
   parent=Switcher_ItemIcon
   widget=Image_ItemIcon
   properties {
    brush {
     bIsDynamicallyLoaded=false
     drawAs=Image
     imageType=NoImage
     mirroring=NoMirror
     resourceName=None
     resourceObject=/Game/Athena/HUD/Materials/M_UI_PingBubble_LargeIItemPing
     tiling=NoTile
     imageSize=<75 75>
     margin=<ltrb 0 0 0 0>
     outlineSettings {
      bUseBrushTransparency=false
      roundingType=HalfHeightRadius
      width=0
      color {colorUseRule=UseColor_Specified, specifiedColor=<rgba 0 0 0 0>}
      cornerRadii=<0 0 0 0>
     }
     tintColor {colorUseRule=UseColor_Specified, specifiedColor=<rgba 1 1 1 1>}
     uVRegion {bIsValid=false, max=<0 0>, min=<0 0>}
    }
    renderTransformPivot=<1 0.5>
   }
   slot {class=CommonVisibilitySwitcherSlot}
  }
  Image_DangerIcon {
   widgetClassPath=&Image
   bIsVariable=true
   parent=Switcher_ItemIcon
   widget=Image_DangerIcon
   properties {
    brush {
     bIsDynamicallyLoaded=false
     drawAs=Image
     imageType=NoImage
     mirroring=NoMirror
     resourceName=None
     resourceObject=/Game/Athena/HUD/Materials/M_UI_PingBubble_Inst
     tiling=NoTile
     imageSize=<75 75>
     margin=<ltrb 0 0 0 0>
     outlineSettings {
      bUseBrushTransparency=false
      roundingType=HalfHeightRadius
      width=0
      color {colorUseRule=UseColor_Specified, specifiedColor=<rgba 0 0 0 0>}
      cornerRadii=<0 0 0 0>
     }
     tintColor {colorUseRule=UseColor_Specified, specifiedColor=<rgba 1 1 1 1>}
     uVRegion {bIsValid=false, max=<0 0>, min=<0 0>}
    }
    renderTransformPivot=<1 0.5>
   }
   slot {class=CommonVisibilitySwitcherSlot, properties {horizontalAlignment=HAlign_Center, verticalAlignment=VAlign_Center}}
  }
  ActionTextContainer {
   widgetClassPath=&HorizontalBox
   bIsVariable=true
   parent=VerticalBox_0
   widget=ActionTextContainer
   slot {class=VerticalBoxSlot, properties {horizontalAlignment=HAlign_Left, padding=<ltrb 14 2 8 0>}}
  }
  ScaleBox_0 {
   widgetClassPath=&ScaleBox
   parent=ActionTextContainer
   widget=ScaleBox_0
   properties {stretch=UserSpecified, userSpecifiedScale=0.6}
   slot {class=HorizontalBoxSlot, properties {verticalAlignment=VAlign_Center}}
  }
  KeybindWidget {
   widgetClassPath=/FortUILibrary/Input/KeybindWidget
   bIsVariable=true
   parent=ScaleBox_0
   widget=KeybindWidget
   properties {boundAction=PlaceMapMarkerInWorld}
   slot {class=ScaleBoxSlot}
  }
  ActionText {
   widgetClassPath=&CommonTextBlock
   bIsVariable=true
   parent=ActionTextContainer
   widget=ActionText
   properties {
    text=Text Block
    textTransformPolicy=ToUpper
    font {
     bForceMonospaced=false
     bMaterialIsStencil=false
     fontMaterial=null
     fontObject=&BurbankSmall
     letterSpacing=0
     monospacedWidth=1
     size=14
     skewAmount=0
     typefaceFontName=Bold
     outlineSettings {
      bApplyOutlineToDropShadows=true
      bMiteredCorners=false
      bSeparateFillAlpha=true
      outlineMaterial=&M-FontShadow
      outlineSize=3
      outlineColor=<rgba 0 0 0 0.5>
     }
    }
    shadowOffset=<0 0>
   }
   slot {class=HorizontalBoxSlot, properties {verticalAlignment=VAlign_Center, padding=<ltrb 4 -2 0 -2>}}
  }
 ]
}
