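# HUD notification widget for Demon Mask Void teleport marker messages, with a leash countdown for the out-of-range warning.
# A stateful event listener feeds HandleMessage, which filters to the owning player, sets styled image, heading and details, then binds to the teleport actor's leash-timer event.
# While the leash is active, a looping timer polls time remaining, injects red rich text into the details' COUNTDOWN token and plays a sound on changes.
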WidgetBlueprint /DemonMaskVoid/UI/TeleportMarker/WBP_DemonMaskVoid_Messaging
graphs [
 BindTurretEvents {nodes [
   n1 = BindTurretEvents!() -> (then, InTurretActor:BGA Demon Mask Void Teleport Base Object Reference) {is=title K2Node_FunctionEntry}
   > UnbindTurretEvents()
   > Set TurretActor(TurretActor:BGA Demon Mask Void Teleport Base Object Reference) -> (Output_Get:BGA Demon Mask Void Teleport Base Object Reference)
   > n2 = Is Valid()
   OnLeashActiveUpdate!(execute=@n2.Is Not Valid, self, IsLeashActive:bool=false) -> (then)
   Bind Event to On Leash Timer Active Updated Event Dispatcher(execute=@n2.Is Valid, self:BGA Demon Mask Void Teleport Base Object Reference=@n1, Delegate=Create Event()) {is=title K2Node_AddDelegate}
  ]}
 EventGraph {nodes [
   Event Construct()
   > Parent: Construct()
   Event Destruct()
   > Parent: Destruct()
   > StopListeningToStatefulEvent(self:Async Action Start Listening To Stateful Event Object Reference=Get MessageListener() -> (MessageListener:Async Action Start Listening To Stateful Event Object Reference))
   Reroute Node()
   > HandleMessage!(execute, self, MessageData:"STRUCT Active Demon Mask Void Teleport Marker Message Data Structure (by ref)") -> (then)
   Set MessageListener(MessageListener:Async Action Start Listening To Stateful Event Object Reference) -> (Output_Get:Async Action Start Listening To Stateful Event Object Reference)
   > PlayIntroAnimation(self:WBP Notification With Backplate Visuals Object Reference=Get WBP_NotificationWithBackplate() -> (WBP_NotificationWithBackplate:WBP Notification With Backplate Visuals Object Reference))
   GetGlobalEventRouter()
  ]}
 HandleMessage {nodes [
   HandleMessage!() -> (then, MessageData:"STRUCT Active Demon Mask Void Teleport Marker Message Data Structure (by ref)") {is=title K2Node_FunctionEntry}
   > Branch(Condition="Equal (Object)"(B=GetOwningPlayer()))
   > SetBrushFromMaterial(self:WBP Notification With Backplate Visuals Object Reference=Get WBP_NotificationWithBackplate())
   > SetHeadling!(
    execute
    self:WBP Notification With Backplate Visuals Object Reference=Get WBP_NotificationWithBackplate()
    InText:Text=Reroute Node(InputPin:Text) -> (OutputPin:Text)
   ) -> (then)
   > SetDetails!(execute, self:WBP Notification With Backplate Visuals Object Reference=Get WBP_NotificationWithBackplate(), InText:Text) -> (then)
   > Set CurrentDetailsText(CurrentDetailsText:Text) -> (Output_Get:Text)
   > BindTurretEvents!(
    execute
    self
    InTurretActor:BGA Demon Mask Void Teleport Base Object Reference=Reroute Node(InputPin:BGA Demon Mask Void Teleport Base Object Reference) -> (OutputPin:BGA Demon Mask Void Teleport Base Object Reference)
   ) -> (then) {is=title K2Node_CallFunction}
   > n1 = Branch(Condition=Reroute Node(InputPin:bool) -> (OutputPin:bool))
   > n2 = Branch(Condition="Equal (Object)"(B=Reroute Node(InputPin:Object Reference) -> (OutputPin:Object Reference)))
   > OnLeashActiveUpdate(IsLeashActive=true)
   OnLeashActiveUpdate(execute=[@n2.else, @n1.else], IsLeashActive=false)
   Get StyleToImageMap() -> (StyleToImageMap:Map of E_DemonMaskVoidMessageStyle Enums to Material Instance Object References=NewEnumerator0)
   Get OutOfRangeMessage() -> (OutOfRangeMessage:BP Demon Mask Void Turret Message Data Type Object Reference)
  ]}
 OnLeashActiveUpdate {nodes [
   n1 = OnLeashActiveUpdate!() -> (then, IsLeashActive:bool=~) {is=title K2Node_FunctionEntry}
   > Branch(Condition="Not Equal (Boolean)"(A=@n1, B=Get IsLeashWidgetActive() -> (IsLeashWidgetActive:bool)) -> (ReturnValue=~))
   > n2 = Set IsLeashWidgetActive(IsLeashWidgetActive:bool=@n1) -> (Output_Get:bool)
   > n3 = Branch(Condition=@n2)
   > UpdateTimeRemaining()
   > n4 = Set Timer by Event(Delegate=Create Event(), Time=0.5, bLooping=true)
   > Set LeashTimerHandle(LeashTimerHandle:Timer Handle Structure=@n4) -> (Output_Get:Timer Handle Structure)
   Clear and Invalidate Timer by Handle(execute=@n3.else, Handle=Get LeashTimerHandle() -> (LeashTimerHandle:Timer Handle Structure))
  ]}
 UnbindTurretEvents {nodes [
   UnbindTurretEvents() {is=title K2Node_FunctionEntry}
   n1 = Get TurretActor() -> (TurretActor:BGA Demon Mask Void Teleport Base Object Reference)
   > Is Valid(InputObject=@n1)
   > Unbind Event from On Leash Timer Active Updated Event Dispatcher(self:BGA Demon Mask Void Teleport Base Object Reference=@n1, Delegate=Create Event()) {is=title K2Node_RemoveDelegate}
   > Set TurretActor()
  ]}
 UpdateTimeRemaining {nodes [
   UpdateTimeRemaining() {is=title K2Node_FunctionEntry}
   > Is Valid(InputObject=Get TurretActor())
   > n1 = Set CurrentTimeRemaining(CurrentTimeRemaining:float=Get CurrentLeashTimeRemaining!(self:BGA Demon Mask Void Teleport Base Object Reference=Get TurretActor()) -> (CurrentLeashTimeRemaining:float)) -> (Output_Get:float)
   > n2 = Branch(Condition="Nearly Equal (Float)"(A=@n1, B=Get PreviousTimeRemaining() -> (PreviousTimeRemaining:float), ErrorTolerance=0))
   Set PreviousTimeRemaining(execute=@n2.else, PreviousTimeRemaining:float=Get CurrentTimeRemaining() -> (CurrentTimeRemaining:float)) -> (Output_Get:float)
   n3 = "To Text (Float)"(Value=Get CurrentTimeRemaining(), MinimumIntegralDigits=1, MaximumIntegralDigits=324, MinimumFractionalDigits=0, MaximumFractionalDigits=3)
   > SetDetails(
    self=Get WBP_NotificationWithBackplate()
    InText=Format Text!(
     Format=Get CurrentDetailsText() -> (CurrentDetailsText:Text)
     COUNTDOWN:Text=Format Text!(Format="INVTEXT(\"<text color=\\\"FF0000FF\\\" size=\\\"18\\\">{value}</>\")", value:Text=@n3) -> (Result)
    ) -> (Result)
   )
   > Branch(Condition="float > float"(A=Get CurrentTimeRemaining()))
   > PlaySound2D(Sound=Get CountdownSoundCue() -> (CountdownSoundCue:Sound Base Object Reference))
  ]}
]
variables [
 CountdownSoundCue {type=SoundBase, default=null}
 CurrentDetailsText {type=string, default=""}
 IsLeashWidgetActive {type=boolean}
 LeashTimerHandle {type=TimerHandle}
 MessageListener {type=AsyncAction_StartListeningToStatefulEvent, default=null}
 OutOfRangeMessage {type=BP_DemonMaskVoid_TurretMessageDataType_C, default=null}
 PreviousTimeRemaining {type=number}
 StyleToImageMap {default {Destroyed=null, PlacementFailure=null, Signal=null, Warning=null}}
 TurretActor {type=BGA_DemonMask_Void_Teleport_Base_C, default=null}
]
widgets {
 info {
  inheritedWidgetCount=0
  parentClass=&CommonUserWidget
  rootWidgetClass=/Game/Athena/HUD/Messaging/WBP_NotificationWithBackplate_Visuals.WBP_NotificationWithBackplate_Visuals_C
  widgetCount=1
 }
 widgets [WBP_NotificationWithBackplate {widgetClassPath=/Game/Athena/HUD/Messaging/WBP_NotificationWithBackplate_Visuals, bIsVariable=true, widget=WBP_NotificationWithBackplate}]
}
